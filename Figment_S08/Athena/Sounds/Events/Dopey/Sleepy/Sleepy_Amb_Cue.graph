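# Looping ambient sound cue for Sleepy in the Dopey event audio, reusing the Snow_Ambient_02 wave from the Snow folder.
# A single wave player feeds the output, routed to the Fort_Events_Mapwide sound class.
# Overridden attenuation combines long-range falloff, distance low-pass and listener focus that lowers off-focus volume.
# Trace-based occlusion fades blocked sound to zero volume with a low-pass filter.
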SoundCue /Figment_S08/Athena/Sounds/Events/Dopey/Sleepy/Sleepy_Amb_Cue
root=WavePlayer_12
nodes [{id=WavePlayer_12, properties {bLooping=true, soundWaveAssetPtr=/Figment_S08/Athena/Sounds/Events/Dopey/Snow/Snow_Ambient_02}}]
properties {
 bOverrideAttenuation=true
 maxDistance=6000
 soundClassObject=/Figment_S08/Athena/Sounds/Events/SoundClasses/Fort_Events_Mapwide
 volumeMultiplier=1
 attenuationOverrides {
  bAttenuateWithLPF=true
  bEnableFocusInterpolation=true
  bEnableListenerFocus=true
  bEnableOcclusion=true
  falloffDistance=6000
  lPFFrequencyAtMax=2000
  lPFRadiusMin=2000
  nonFocusVolumeAttenuation=0.5
  occlusionInterpolationTime=1
  occlusionLowPassFilterFrequency=1000
  occlusionTraceChannel=ECC_GameTraceChannel11
  occlusionVolumeAttenuation=0
  reverbDistanceMin=0
  reverbWetLevelMax=20
  reverbWetLevelMin=0
  stereoSpread=512
  attenuationShapeExtents {x=0}
 }
 concurrencyOverrides {maxCount=4, resolutionRule=StopOldest}
}
