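# Sound cue for 2D Sleepy impact sounds in the Figment_S08 Dopey event, playing short, simple metal hits.
# A switch on the integer Size parameter sends all four inputs to the same modulator, which feeds an equally weighted random pick of six waves.
# It borrows the Snow impacts 2D attenuation, uses a ducking impact sound class, boosts volume and overrides concurrency to stop the oldest instance.
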SoundCue /Figment_S08/Athena/Sounds/Events/Dopey/Sleepy/Sleepy_Impact_2D_Cue
root=Switch_1
nodes [
 {id=Modulator_0, children [Random_2]}
 {
  id=Random_2
  children [WavePlayer_3, WavePlayer_4, WavePlayer_8, WavePlayer_9, WavePlayer_16, WavePlayer_17]
  properties {weights [1,1,1,1,1,1]}
 }
 {id=Switch_1, children [Modulator_0,Modulator_0,Modulator_0,Modulator_0], properties {intParameterName=Size}}
 {id=WavePlayer_16, properties {soundWaveAssetPtr=&/Sleepy_Impact_Metal_Short_Simple_05}}
 {id=WavePlayer_17, properties {soundWaveAssetPtr=&/Sleepy_Impact_Metal_Short_Simple_06}}
 {id=WavePlayer_3, properties {soundWaveAssetPtr=&/Sleepy_Impact_Metal_Short_Simple_01}}
 {id=WavePlayer_4, properties {soundWaveAssetPtr=&/Sleepy_Impact_Metal_Short_Simple_02}}
 {id=WavePlayer_8, properties {soundWaveAssetPtr=&/Sleepy_Impact_Metal_Short_Simple_03}}
 {id=WavePlayer_9, properties {soundWaveAssetPtr=&/Sleepy_Impact_Metal_Short_Simple_04}}
]
properties {
 attenuationSettings=/Figment_S08/Athena/Sounds/Events/Dopey/Snow/Snow_Impacts_2D_Att
 bOverrideConcurrency=true
 maxDistance=2097152
 soundClassObject=/Game/Sounds/FortSoundClassesAndModes/Fort_ImpactSounds_DuckDefaultQuieter_PSM
 volumeMultiplier=2
 attenuationOverrides {falloffDistance=3000, attenuationShapeExtents {x=0}}
 concurrencyOverrides {maxCount=4, resolutionRule=StopOldest}
}
