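# Sound cue for Sleepy's long resonant metal impact in the Figment_S08 Dopey event sounds, played at reduced volume through an impact sound class.
# A Size integer switch sends all four inputs to one shared Modulator, so every size sounds alike.
# That Modulator feeds an equally weighted Random node over six wave variations.
# It borrows the Snow impacts 2D attenuation asset and overrides concurrency to block new plays beyond two.
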SoundCue /Figment_S08/Athena/Sounds/Events/Dopey/Sleepy/Sleepy_Impact_2D_Resonant_Cue
root=Switch_0
nodes [
 {id=Modulator_1, children [Random_0]}
 {
  id=Random_0
  children [WavePlayer_5, WavePlayer_6, WavePlayer_7, WavePlayer_10, WavePlayer_11, WavePlayer_12]
  properties {weights [1,1,1,1,1,1]}
 }
 {id=Switch_0, children [Modulator_1,Modulator_1,Modulator_1,Modulator_1], properties {intParameterName=Size}}
 {id=WavePlayer_10, properties {soundWaveAssetPtr=&/Sleepy_Impact_Metal_Long_Resonant_04}}
 {id=WavePlayer_11, properties {soundWaveAssetPtr=&/Sleepy_Impact_Metal_Long_Resonant_05}}
 {id=WavePlayer_12, properties {soundWaveAssetPtr=&/Sleepy_Impact_Metal_Long_Resonant_06}}
 {id=WavePlayer_5, properties {soundWaveAssetPtr=&/Sleepy_Impact_Metal_Long_Resonant_01}}
 {id=WavePlayer_6, properties {soundWaveAssetPtr=&/Sleepy_Impact_Metal_Long_Resonant_02}}
 {id=WavePlayer_7, properties {soundWaveAssetPtr=&/Sleepy_Impact_Metal_Long_Resonant_03}}
]
properties {
 attenuationSettings=/Figment_S08/Athena/Sounds/Events/Dopey/Snow/Snow_Impacts_2D_Att
 bOverrideConcurrency=true
 maxDistance=2097152
 soundClassObject=/Game/Sounds/FortSoundClassesAndModes/Fort_ImpactSounds_DuckDefaultQuieter_PSM
 volumeMultiplier=0.25
 attenuationOverrides {falloffDistance=3000, attenuationShapeExtents {x=0}}
 concurrencyOverrides {maxCount=2, resolutionRule=PreventNew}
}
