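# Sound cue for the Sleepy pre-complete start sound in the Figment S08 Dopey event audio set.
# A single wave player plays a Snow pre-complete start wave, so the cue reuses Snow assets despite its Sleepy name.
# It uses the Snow close-distant attenuation, the map-wide events sound class, boosted volume, and a concurrency override that stops the oldest instance at the limit.
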SoundCue /Figment_S08/Athena/Sounds/Events/Dopey/Sleepy/Sleepy_Pre_Complete_Start_Cue
root=WavePlayer_0
nodes [{id=WavePlayer_0, properties {soundWaveAssetPtr=/Figment_S08/Athena/Sounds/Events/Dopey/Snow/Snow_Pre_Complete_Start_01}}]
properties {
 attenuationSettings=/Figment_S08/Athena/Sounds/Events/Dopey/Snow/Snow_MLTI_CloseDistant_Att
 bOverrideConcurrency=true
 maxDistance=2097152
 soundClassObject=/Figment_S08/Athena/Sounds/Events/SoundClasses/Fort_Events_Mapwide
 volumeMultiplier=1.5
 attenuationOverrides {falloffDistance=3000, attenuationShapeExtents {x=0}}
 concurrencyOverrides {maxCount=4, resolutionRule=StopOldest}
}
